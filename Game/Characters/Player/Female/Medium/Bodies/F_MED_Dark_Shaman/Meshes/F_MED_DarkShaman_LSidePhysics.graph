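# Physics asset for the left-side skirt of the medium female Dark Shaman player body, covering its dynamic skirt bones.
# Capsules on the pelvis, thighs and calves form the lower body, while boxes form front, side and back skirt chains.
# Each chain starts at a small base box joined to the left thigh with free rotation, and the panels below use limited swing and twist, tighter sideways and in twist.
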PhysicsAsset /Game/Characters/Player/Female/Medium/Bodies/F_MED_Dark_Shaman/Meshes/F_MED_DarkShaman_LSidePhysics
bodies {
 calf_l [{shapeType=Capsule, length=34.7288, radius=6.6101, center=<-23.6751 -0.0115 1.8284>, rotation=<rot 90 0 0>}]
 calf_r [{shapeType=Capsule, length=34.7826, radius=5.6179, center=<23.7456 0 -0.0004>, rotation=<rot 90 0 0>}]
 dyn_skirtBackBase_l [{shapeType=Box, extentX=1.01, extentY=1.01, extentZ=1.01, rotation=<rot 0 -90 0>}]
 dyn_skirtBack_l [{shapeType=Box, extentX=2.8797, extentY=18.8207, extentZ=9.9333, center=<-8.4838 -0.0801 2.8701>, rotation=<rot -4.9654 86.8577 167.607>}]
 dyn_skirtFrontBase_01_l [{shapeType=Box, extentX=1.01, extentY=1.01, extentZ=1.01, rotation=<rot 0 -90 0>}]
 dyn_skirtFront_01_l [{shapeType=Box, extentX=3.0054, extentY=9.3929, extentZ=9.7296, center=<-4.6199 -0.0001 0>, rotation=<rot 0 90 90>}]
 dyn_skirtFront_02_l [{shapeType=Box, extentX=3.0917, extentY=7.8335, extentZ=9.1015, center=<-2.9873 0 0>, rotation=<rot 0 90 90>}]
 dyn_skirt_01Base_l [{shapeType=Box, extentX=1.01, extentY=1.01, extentZ=1.01, rotation=<rot 0 -90 0>}]
 dyn_skirt_01_l [{shapeType=Box, extentX=3.3297, extentY=7.4164, extentZ=10.2466, center=<-4.692 0 0.1286>, rotation=<rot 0 89.9999 86.1003>}]
 dyn_skirt_02_l [{shapeType=Box, extentX=4.7573, extentY=9.6111, extentZ=10.4177, center=<-4.6783 0 0>, rotation=<rot 0 90 90>}]
 dyn_skirt_03_l [{shapeType=Box, extentX=5.5224, extentY=10.144, extentZ=11.5645, center=<-5.6387 0.2547 0.8267>, rotation=<rot 0.2384 86.1615 93.5495>}]
 pelvis [{shapeType=Capsule, length=9.6919, radius=10.9429, center=<-2.6895 -2.093 -0.053>, rotation=<rot -3.0486 -90 -0.0239>}]
 thigh_l [{shapeType=Capsule, length=23.961, radius=7.8659, center=<-26.2203 1.0122 0.5825>, rotation=<rot 90 0 0>}]
 thigh_r [{shapeType=Capsule, length=22.4463, radius=7.4128, center=<25.1845 0 -0.0003>, rotation=<rot 90 0 0>}]
}
constraints [
 {bone1Name=calf_l, bone2Name=thigh_l, swing1LimitDegrees=45, swing2LimitDegrees=45, twistLimitDegrees=45}
 {bone1Name=calf_r, bone2Name=thigh_r, swing1LimitDegrees=45, swing2LimitDegrees=45, twistLimitDegrees=45}
 {bone1Name=dyn_skirt_01_l, bone2Name=dyn_skirt_01Base_l, swing1LimitDegrees=45, swing2LimitDegrees=10, twistLimitDegrees=10}
 {bone1Name=dyn_skirt_02_l, bone2Name=dyn_skirt_01_l, swing1LimitDegrees=45, swing2LimitDegrees=20, twistLimitDegrees=10}
 {bone1Name=dyn_skirt_03_l, bone2Name=dyn_skirt_02_l, swing1LimitDegrees=45, swing2LimitDegrees=20, twistLimitDegrees=10}
 {bone1Name=dyn_skirtBack_l, bone2Name=dyn_skirtBackBase_l, swing1LimitDegrees=45, swing2LimitDegrees=20, twistLimitDegrees=10}
 {bone1Name=dyn_skirtFront_01_l, bone2Name=dyn_skirtFrontBase_01_l, swing1LimitDegrees=45, swing2LimitDegrees=10, twistLimitDegrees=10}
 {bone1Name=dyn_skirtFront_02_l, bone2Name=dyn_skirtFront_01_l, swing1LimitDegrees=45, swing2LimitDegrees=20, twistLimitDegrees=10}
 {
  bone1Name=dyn_skirtBackBase_l
  bone2Name=thigh_l
  swing1LimitDegrees=45
  swing1Motion=Free
  swing2LimitDegrees=45
  swing2Motion=Free
  twistLimitDegrees=45
  twistMotion=Free
 }
 {
  bone1Name=dyn_skirt_01Base_l
  bone2Name=thigh_l
  swing1LimitDegrees=45
  swing1Motion=Free
  swing2LimitDegrees=45
  swing2Motion=Free
  twistLimitDegrees=45
  twistMotion=Free
 }
 {
  bone1Name=dyn_skirtFrontBase_01_l
  bone2Name=thigh_l
  swing1LimitDegrees=45
  swing1Motion=Free
  swing2LimitDegrees=45
  swing2Motion=Free
  twistLimitDegrees=45
  twistMotion=Free
 }
 {
  bone1Name=thigh_l
  bone2Name=pelvis
  swing1LimitDegrees=45
  swing1Motion=Free
  swing2LimitDegrees=45
  swing2Motion=Free
  twistLimitDegrees=45
  twistMotion=Free
 }
 {
  bone1Name=thigh_r
  bone2Name=pelvis
  swing1LimitDegrees=45
  swing1Motion=Free
  swing2LimitDegrees=45
  swing2Motion=Free
  twistLimitDegrees=45
  twistMotion=Free
 }
]
