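# Physics asset for the female medium Dark Shaman's loincloth and its bells, giving them simulated secondary motion.
# Two five-box chains hang from a pelvis-linked flap root, with swing/twist limits along each and looser cross-links between the lower four pairs so they move as one sheet.
# Pelvis, thigh and calf capsules supply body shapes for the cloth.
# Bell spheres are only partly constrained: bells 03, 05 and 07 have no constraints.
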PhysicsAsset /Game/Characters/Player/Female/Medium/Bodies/F_MED_Dark_Shaman/Meshes/F_MED_DarkShaman_Loincloth_Physics
bodies {
 calf_l [{shapeType=Capsule, length=34.7288, radius=6.6101, center=<-23.6751 -0.0115 1.8284>, rotation=<rot 90 0 0>}]
 calf_r [{shapeType=Capsule, length=34.7826, radius=5.6179, center=<23.7456 0 -0.0004>, rotation=<rot 90 0 0>}]
 dyn_bellBase [{shapeType=Sphere, radius=0.505}]
 dyn_bell_01 [{shapeType=Sphere, radius=0.8033}]
 dyn_bell_010 [{shapeType=Sphere, radius=0.801, center=<-0.8718 0.134 0.1221>}]
 dyn_bell_03 [{shapeType=Sphere, radius=0.8034}]
 dyn_bell_05 [{shapeType=Sphere, radius=0.8034}]
 dyn_bell_07 [{shapeType=Sphere, radius=0.8033}]
 dyn_bell_09 [{shapeType=Sphere, radius=0.505}]
 dyn_loincloth_01_l [{shapeType=Box, extentX=1.7954, extentY=7.0135, extentZ=8.4348, center=<2.9156 0.0001 0>, rotation=<rot 0 90 0>}]
 dyn_loincloth_01_r [{shapeType=Box, extentX=2.8445, extentY=7.7983, extentZ=9.3402, center=<3.2447 -0.8541 0.0122>, rotation=<rot 3.8573 -90 1.8071>}]
 dyn_loincloth_02_l [{shapeType=Box, extentX=2.0233, extentY=8.6009, extentZ=7.9557, center=<3.8097 0.0004 0>, rotation=<rot 0 90 0>}]
 dyn_loincloth_02_r [{shapeType=Box, extentX=1.7458, extentY=8.4064, extentZ=8.7215, center=<3.7845 0 0>, rotation=<rot 0 90 0>}]
 dyn_loincloth_03_l [{shapeType=Box, extentX=1.567, extentY=8.6828, extentZ=6.6527, center=<3.8816 0.0001 0>, rotation=<rot 0 90 0>}]
 dyn_loincloth_03_r [{shapeType=Box, extentX=1.6347, extentY=8.7903, extentZ=8.6115, center=<4.0273 0.0002 0>, rotation=<rot 0 90 0>}]
 dyn_loincloth_04_l [{shapeType=Box, extentX=1.5373, extentY=9.5302, extentZ=6.9303, center=<4.232 -0.0001 0>, rotation=<rot 0 90 0>}]
 dyn_loincloth_04_r [{shapeType=Box, extentX=1.5457, extentY=10.201, extentZ=8.7761, center=<4.6023 0.0001 0>, rotation=<rot 0 90 0>}]
 dyn_loincloth_05_l [{shapeType=Box, extentX=1.6133, extentY=17.568, extentZ=6.8152, center=<8.1471 0.0002 0>, rotation=<rot 0 90 0>}]
 dyn_loincloth_05_r [{shapeType=Box, extentX=1.5945, extentY=13.4851, extentZ=9.0558, center=<6.2918 0.0001 0>, rotation=<rot 0 90 0>}]
 dyn_loincloth_FlapRoot [{shapeType=Box, extentX=2.6497, extentY=2.2206, extentZ=16.7068, center=<0.7976 -0.4319 0.2225>, rotation=<rot 0.5134 -90 -0.6541>}]
 pelvis [{shapeType=Capsule, length=9.6919, radius=10.9429, center=<-2.6895 -2.093 -0.053>, rotation=<rot -3.0486 -90 -0.0239>}]
 thigh_l [{shapeType=Capsule, length=31.5217, radius=7.8659, center=<-20.9455 0.7949 0.5825>, rotation=<rot 90 0 0>}]
 thigh_r [{shapeType=Capsule, length=33.8049, radius=7.4128, center=<19.626 0 -0.0003>, rotation=<rot 90 0 0>}]
}
constraints [
 {bone1Name=calf_l, bone2Name=thigh_l, swing1LimitDegrees=45, swing2LimitDegrees=45, twistLimitDegrees=45}
 {bone1Name=calf_r, bone2Name=thigh_r, swing1LimitDegrees=45, swing2LimitDegrees=45, twistLimitDegrees=45}
 {
  bone1Name=thigh_l
  bone2Name=pelvis
  swing1LimitDegrees=45
  swing1Motion=Free
  swing2LimitDegrees=45
  swing2Motion=Free
  twistLimitDegrees=45
  twistMotion=Free
 }
 {
  bone1Name=thigh_r
  bone2Name=pelvis
  swing1LimitDegrees=45
  swing1Motion=Free
  swing2LimitDegrees=45
  swing2Motion=Free
  twistLimitDegrees=45
  twistMotion=Free
 }
 {
  bone1Name=dyn_bell_01
  bone2Name=dyn_bellBase
  swing1LimitDegrees=45
  swing1Motion=Free
  swing2LimitDegrees=45
  swing2Motion=Free
  twistLimitDegrees=45
  twistMotion=Free
 }
 {bone1Name=dyn_bell_010, bone2Name=dyn_bell_09, swing1LimitDegrees=20, swing2LimitDegrees=20, twistLimitDegrees=5}
 {
  bone1Name=dyn_bellBase
  bone2Name=pelvis
  swing1LimitDegrees=45
  swing1Motion=Free
  swing2LimitDegrees=45
  swing2Motion=Free
  twistLimitDegrees=45
  twistMotion=Free
 }
 {
  bone1Name=dyn_loincloth_FlapRoot
  bone2Name=pelvis
  swing1LimitDegrees=45
  swing1Motion=Free
  swing2LimitDegrees=45
  swing2Motion=Free
  twistLimitDegrees=45
  twistMotion=Free
 }
 {bone1Name=dyn_loincloth_01_r, bone2Name=dyn_loincloth_FlapRoot, swing1LimitDegrees=30, swing2LimitDegrees=30, twistLimitDegrees=10}
 {bone1Name=dyn_loincloth_02_r, bone2Name=dyn_loincloth_01_r, swing1LimitDegrees=30, swing2LimitDegrees=30, twistLimitDegrees=10}
 {bone1Name=dyn_loincloth_03_r, bone2Name=dyn_loincloth_02_r, swing1LimitDegrees=30, swing2LimitDegrees=30, twistLimitDegrees=10}
 {bone1Name=dyn_loincloth_04_r, bone2Name=dyn_loincloth_03_r, swing1LimitDegrees=30, swing2LimitDegrees=30, twistLimitDegrees=10}
 {bone1Name=dyn_loincloth_05_r, bone2Name=dyn_loincloth_04_r, swing1LimitDegrees=30, swing2LimitDegrees=30, twistLimitDegrees=10}
 {bone1Name=dyn_loincloth_01_l, bone2Name=dyn_loincloth_FlapRoot, swing1LimitDegrees=30, swing2LimitDegrees=30, twistLimitDegrees=10}
 {bone1Name=dyn_loincloth_02_l, bone2Name=dyn_loincloth_01_l, swing1LimitDegrees=30, swing2LimitDegrees=30, twistLimitDegrees=10}
 {bone1Name=dyn_loincloth_03_l, bone2Name=dyn_loincloth_02_l, swing1LimitDegrees=30, swing2LimitDegrees=30, twistLimitDegrees=10}
 {bone1Name=dyn_loincloth_04_l, bone2Name=dyn_loincloth_03_l, swing1LimitDegrees=30, swing2LimitDegrees=30, twistLimitDegrees=10}
 {bone1Name=dyn_loincloth_05_l, bone2Name=dyn_loincloth_04_l, swing1LimitDegrees=30, swing2LimitDegrees=30, twistLimitDegrees=10}
 {bone1Name=dyn_loincloth_05_l, bone2Name=dyn_loincloth_05_r, swing1LimitDegrees=40, swing2LimitDegrees=40, twistLimitDegrees=20}
 {bone1Name=dyn_loincloth_04_l, bone2Name=dyn_loincloth_04_r, swing1LimitDegrees=40, swing2LimitDegrees=40, twistLimitDegrees=20}
 {bone1Name=dyn_loincloth_03_l, bone2Name=dyn_loincloth_03_r, swing1LimitDegrees=40, swing2LimitDegrees=40, twistLimitDegrees=20}
 {bone1Name=dyn_loincloth_02_l, bone2Name=dyn_loincloth_02_r, swing1LimitDegrees=40, swing2LimitDegrees=40, twistLimitDegrees=20}
]
