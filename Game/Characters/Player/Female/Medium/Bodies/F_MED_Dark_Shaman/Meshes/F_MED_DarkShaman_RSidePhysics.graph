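# Physics asset for the female medium Dark Shaman player body, rigging its skirt, loincloth and bell chain for secondary motion against pelvis and leg capsules.
# Box-bodied skirt panels on both sides hang from angularly free base bodies on each thigh, with the first side panel nearly locked to its base.
# Capsule chains form the pelvis-hung bells and two loincloth strips, which are cross-linked at their second and last segments by angularly free constraints.
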PhysicsAsset /Game/Characters/Player/Female/Medium/Bodies/F_MED_Dark_Shaman/Meshes/F_MED_DarkShaman_RSidePhysics
bodies {
 calf_l [{shapeType=Capsule, length=34.4049, radius=6.9339, center=<-23.6751 -3.5521 1.8284>, rotation=<rot 90 0 0>}]
 calf_r [{shapeType=Capsule, length=33.2535, radius=7.147, center=<23.7456 2.7928 -0.0005>, rotation=<rot 90 0 0>}]
 dyn_bellBase [{shapeType=Capsule, length=0.505, radius=0.505, shapeName=dyn_bellBase_capsule, rotation=<rot 0 -90 0>}]
 dyn_bell_01 [{
   shapeType=Capsule
   length=1.0643
   radius=1.7211
   shapeName=dyn_bell_01_capsule
   center=<1.1124 0.0005 -0.0257>
   rotation=<rot 0 -89.9487 -85.9496>
  }]
 dyn_bell_010 [{
   shapeType=Capsule
   length=3.1389
   radius=1.8611
   shapeName=dyn_bell_010_capsule
   center=<3.5611 -0.0479 0.0014>
   rotation=<rot 90 13.4732 16.7405>
  }]
 dyn_bell_03 [{
   shapeType=Capsule
   length=1.1389
   radius=1.8611
   shapeName=dyn_bell_03_capsule
   center=<1.7991 0.0001 0.0041>
   rotation=<rot 89.8703 179.9788 179.9788>
  }]
 dyn_bell_05 [{
   shapeType=Capsule
   length=1.1389
   radius=1.8611
   shapeName=dyn_bell_05_capsule
   center=<1.7991 0.0001 0.0038>
   rotation=<rot 89.878 179.9814 179.9814>
  }]
 dyn_bell_07 [{
   shapeType=Capsule
   length=1.1389
   radius=1.8611
   shapeName=dyn_bell_07_capsule
   center=<1.7991 0.0001 0.0041>
   rotation=<rot 89.8703 179.9815 179.9815>
  }]
 dyn_bell_09 [{
   shapeType=Capsule
   length=1.1389
   radius=1.8611
   shapeName=dyn_bell_09_capsule
   center=<1.7991 0.1456 0.0038>
   rotation=<rot 89.878 179.9784 179.9784>
  }]
 dyn_loincloth_01_l [{
   shapeType=Capsule
   length=3.5
   radius=3.5
   shapeName=dyn_loincloth_01_l_capsule
   center=<3.4152 -0.0001 0.0597>
   rotation=<rot 88.9985 179.9999 180>
  }]
 dyn_loincloth_01_r [{
   shapeType=Capsule
   length=3.5
   radius=3.5
   shapeName=dyn_loincloth_01_r_capsule
   center=<3.4152 -0.0001 0.0594>
   rotation=<rot 89.0034 179.9999 180>
  }]
 dyn_loincloth_02_l [{
   shapeType=Capsule
   length=6.5
   radius=3.5
   shapeName=dyn_loincloth_02_l_capsule
   center=<5.0411 -0.0002 0.0881>
   rotation=<rot 88.9985 179.9999 180>
  }]
 dyn_loincloth_02_r [{
   shapeType=Capsule
   length=6.5
   radius=3.5
   shapeName=dyn_loincloth_02_r_capsule
   center=<5.0411 -0.0002 0.0874>
   rotation=<rot 89.0067 179.9999 180>
  }]
 dyn_loincloth_03_l [{
   shapeType=Capsule
   length=7.1938
   radius=2.8062
   shapeName=dyn_loincloth_03_l_capsule
   center=<5.0411 -0.0002 0.0877>
   rotation=<rot 89.0034 179.9999 180>
  }]
 dyn_loincloth_03_r [{
   shapeType=Capsule
   length=7.1938
   radius=2.8062
   shapeName=dyn_loincloth_03_r_capsule
   center=<5.0411 -0.0002 0.0877>
   rotation=<rot 89.0034 179.9999 180>
  }]
 dyn_loincloth_04_l [{
   shapeType=Capsule
   length=7.4202
   radius=2.5798
   shapeName=dyn_loincloth_04_l_capsule
   center=<4.7374 -0.0001 0.0824>
   rotation=<rot 89.0034 179.9999 -180>
  }]
 dyn_loincloth_04_r [{
   shapeType=Capsule
   length=7.4202
   radius=2.5798
   shapeName=dyn_loincloth_04_r_capsule
   center=<4.7374 -0.0001 0.0828>
   rotation=<rot 88.9985 179.9999 -180>
  }]
 dyn_loincloth_05_l [{
   shapeType=Capsule
   length=7.4202
   radius=2.5798
   shapeName=dyn_loincloth_05_l_capsule
   center=<4.7374 -0.0001 0.0824>
   rotation=<rot 89.0034 179.9999 -180>
  }]
 dyn_loincloth_05_r [{
   shapeType=Capsule
   length=8.1033
   radius=1.8967
   shapeName=dyn_loincloth_05_r_capsule
   center=<4.7374 -0.0001 0.0824>
   rotation=<rot 89.0034 179.9999 -180>
  }]
 dyn_skirtBackBase_l [{shapeType=Box, extentX=1.01, extentY=1.01, extentZ=1.01, rotation=<rot 0 -90 0>}]
 dyn_skirtBackBase_r [{shapeType=Box, extentX=1.01, extentY=1.01, extentZ=1.01, rotation=<rot 0 -90 0>}]
 dyn_skirtBack_l [{shapeType=Box, extentX=4.4213, extentY=18.8207, extentZ=12.2468, center=<-8.1832 1.3397 1.7974>, rotation=<rot 4.7062 -95.2512 -168.4192>}]
 dyn_skirtBack_r [{shapeType=Box, extentX=2.8797, extentY=18.8207, extentZ=9.9333, center=<8.551 0.265 -1.6868>, rotation=<rot -4.6545 -94.6437 -168.2229>}]
 dyn_skirtFrontBase_01_l [{shapeType=Box, extentX=1.01, extentY=1.01, extentZ=1.01, rotation=<rot 0 -90 0>}]
 dyn_skirtFrontBase_01_r [{shapeType=Box, extentX=1.01, extentY=1.01, extentZ=1.01, rotation=<rot 0 -90 0>}]
 dyn_skirtFront_01_l [{shapeType=Box, extentX=3.5842, extentY=9.3929, extentZ=9.7296, center=<-4.6199 -0.0001 0>, rotation=<rot 0 90 90>}]
 dyn_skirtFront_01_r [{shapeType=Box, extentX=2.5307, extentY=9.3929, extentZ=9.7296, center=<2.5564 0.0001 0>, rotation=<rot 0 -90 89.9999>}]
 dyn_skirtFront_02_l [{shapeType=Box, extentX=4.2864, extentY=11.3739, extentZ=9.1015, center=<-1.4418 -0.6088 0.2216>, rotation=<rot -8.5452 89.9999 89.9999>}]
 dyn_skirtFront_02_r [{shapeType=Box, extentX=3.5422, extentY=7.8335, extentZ=9.1015, center=<0.9238 0 0>, rotation=<rot 0 -90 89.9999>}]
 dyn_skirt_01Base_l [{shapeType=Box, extentX=1.01, extentY=1.01, extentZ=1.01, rotation=<rot 0 -90 0>}]
 dyn_skirt_01Base_r [{shapeType=Box, extentX=1.01, extentY=1.01, extentZ=1.01, rotation=<rot 0 -90 0>}]
 dyn_skirt_01_l [{shapeType=Box, extentX=2.9961, extentY=7.4164, extentZ=10.2466, center=<-4.692 0 0.1286>, rotation=<rot 0 89.9999 86.1003>}]
 dyn_skirt_01_r [{shapeType=Box, extentX=2.5286, extentY=7.4164, extentZ=10.2466, center=<4.692 0 0.1286>, rotation=<rot 0 -90.0001 86.1003>}]
 dyn_skirt_02_l [{shapeType=Box, extentX=4.9569, extentY=13.355, extentZ=11.4161, center=<-3.5909 0.0001 0>, rotation=<rot 0 -86.3146 89.9999>}]
 dyn_skirt_02_r [{shapeType=Box, extentX=3.8421, extentY=9.6111, extentZ=10.4177, center=<4.6783 0 0>, rotation=<rot 0 -87.0399 89.9999>}]
 dyn_skirt_03_l [{shapeType=Box, extentX=4.4643, extentY=11.023, extentZ=11.5645, center=<-2.4304 -0.3438 1.0262>, rotation=<rot -0.2977 94.7942 93.5449>}]
 dyn_skirt_03_r [{shapeType=Box, extentX=3.2968, extentY=10.144, extentZ=11.5645, center=<5.636 -0.0253 0.8269>, rotation=<rot -0.0434 -89.3029 83.9791>}]
 pelvis [{shapeType=Capsule, length=12.4419, radius=9.9429, center=<-2.9524 -3.2322 -0.0564>, rotation=<rot -3.0486 -89.9994 178.8025>}]
 thigh_l [{shapeType=Capsule, length=23.5101, radius=8.3168, center=<-26.2203 1.0122 0.9018>, rotation=<rot 90 0 0>}]
 thigh_r [{shapeType=Capsule, length=24.1963, radius=8.1628, center=<26.583 -0.0001 -0.789>, rotation=<rot 90 0 0>}]
 thigh_twist_01_r [{
   shapeType=Capsule
   length=31.3205
   radius=9.25
   shapeName=thigh_twist_01_r_capsule
   center=<-8.27 0.6397 0.7481>
   rotation=<rot -82.5813 0 0>
  }]
}
constraints [
 {bone1Name=calf_l, bone2Name=thigh_l, swing1LimitDegrees=45, swing2LimitDegrees=45, twistLimitDegrees=45}
 {bone1Name=calf_r, bone2Name=thigh_r, swing1LimitDegrees=45, swing2LimitDegrees=45, twistLimitDegrees=45}
 {bone1Name=dyn_skirt_01_l, bone2Name=dyn_skirt_01Base_l, swing1LimitDegrees=0.1, swing2LimitDegrees=0.1, twistLimitDegrees=0.1}
 {bone1Name=dyn_skirt_02_l, bone2Name=dyn_skirt_01_l, swing1LimitDegrees=45, swing2LimitDegrees=20, twistLimitDegrees=10}
 {bone1Name=dyn_skirt_03_l, bone2Name=dyn_skirt_02_l, swing1LimitDegrees=45, swing2LimitDegrees=20, twistLimitDegrees=10}
 {bone1Name=dyn_skirtBack_l, bone2Name=dyn_skirtBackBase_l, swing1LimitDegrees=45, swing2LimitDegrees=20, twistLimitDegrees=10}
 {bone1Name=dyn_skirtFront_01_l, bone2Name=dyn_skirtFrontBase_01_l, swing1LimitDegrees=45, swing2LimitDegrees=10, twistLimitDegrees=10}
 {bone1Name=dyn_skirtFront_02_l, bone2Name=dyn_skirtFront_01_l, swing1LimitDegrees=45, swing2LimitDegrees=20, twistLimitDegrees=10}
 {
  bone1Name=dyn_skirtBackBase_l
  bone2Name=thigh_l
  swing1LimitDegrees=45
  swing1Motion=Free
  swing2LimitDegrees=45
  swing2Motion=Free
  twistLimitDegrees=45
  twistMotion=Free
 }
 {
  bone1Name=dyn_skirt_01Base_l
  bone2Name=thigh_l
  swing1LimitDegrees=45
  swing1Motion=Free
  swing2LimitDegrees=45
  swing2Motion=Free
  twistLimitDegrees=45
  twistMotion=Free
 }
 {
  bone1Name=dyn_skirtFrontBase_01_l
  bone2Name=thigh_l
  swing1LimitDegrees=45
  swing1Motion=Free
  swing2LimitDegrees=45
  swing2Motion=Free
  twistLimitDegrees=45
  twistMotion=Free
 }
 {
  bone1Name=thigh_l
  bone2Name=pelvis
  swing1LimitDegrees=45
  swing1Motion=Free
  swing2LimitDegrees=45
  swing2Motion=Free
  twistLimitDegrees=45
  twistMotion=Free
 }
 {
  bone1Name=thigh_r
  bone2Name=pelvis
  swing1LimitDegrees=45
  swing1Motion=Free
  swing2LimitDegrees=45
  swing2Motion=Free
  twistLimitDegrees=45
  twistMotion=Free
 }
 {
  bone1Name=dyn_skirt_01Base_r
  bone2Name=thigh_r
  swing1LimitDegrees=45
  swing1Motion=Free
  swing2LimitDegrees=45
  swing2Motion=Free
  twistLimitDegrees=45
  twistMotion=Free
 }
 {
  bone1Name=dyn_skirtFrontBase_01_r
  bone2Name=thigh_r
  swing1LimitDegrees=45
  swing1Motion=Free
  swing2LimitDegrees=45
  swing2Motion=Free
  twistLimitDegrees=45
  twistMotion=Free
 }
 {
  bone1Name=dyn_skirtBackBase_r
  bone2Name=thigh_r
  swing1LimitDegrees=45
  swing1Motion=Free
  swing2LimitDegrees=45
  swing2Motion=Free
  twistLimitDegrees=45
  twistMotion=Free
 }
 {bone1Name=dyn_skirtBack_r, bone2Name=dyn_skirtBackBase_r, swing1LimitDegrees=45, swing2LimitDegrees=20, twistLimitDegrees=10}
 {bone1Name=dyn_skirt_01_r, bone2Name=dyn_skirt_01Base_r, swing1LimitDegrees=0.1, swing2LimitDegrees=0.1, twistLimitDegrees=0.1}
 {bone1Name=dyn_skirt_02_r, bone2Name=dyn_skirt_01_r, swing1LimitDegrees=45, swing2LimitDegrees=20, twistLimitDegrees=10}
 {bone1Name=dyn_skirt_03_r, bone2Name=dyn_skirt_02_r, swing1LimitDegrees=45, swing2LimitDegrees=20, twistLimitDegrees=10}
 {bone1Name=dyn_skirtFront_01_r, bone2Name=dyn_skirtFrontBase_01_r, swing1LimitDegrees=45, swing2LimitDegrees=10, twistLimitDegrees=10}
 {bone1Name=dyn_skirtFront_02_r, bone2Name=dyn_skirtFront_01_r, swing1LimitDegrees=45, swing2LimitDegrees=20, twistLimitDegrees=10}
 {bone1Name=dyn_loincloth_05_r, bone2Name=dyn_loincloth_04_r, swing1LimitDegrees=45, swing2LimitDegrees=15, twistLimitDegrees=15}
 {bone1Name=dyn_loincloth_04_r, bone2Name=dyn_loincloth_03_r, swing1LimitDegrees=45, swing2LimitDegrees=15, twistLimitDegrees=15}
 {bone1Name=dyn_loincloth_03_r, bone2Name=dyn_loincloth_02_r, swing1LimitDegrees=45, swing2LimitDegrees=15, twistLimitDegrees=15}
 {bone1Name=dyn_loincloth_02_r, bone2Name=dyn_loincloth_01_r, swing1LimitDegrees=45, swing2LimitDegrees=15, twistLimitDegrees=15}
 {bone1Name=dyn_loincloth_01_r, bone2Name=pelvis, swing1LimitDegrees=45, swing2LimitDegrees=15, twistLimitDegrees=15}
 {bone1Name=dyn_loincloth_05_l, bone2Name=dyn_loincloth_04_l, swing1LimitDegrees=45, swing2LimitDegrees=15, twistLimitDegrees=15}
 {bone1Name=dyn_loincloth_04_l, bone2Name=dyn_loincloth_03_l, swing1LimitDegrees=45, swing2LimitDegrees=15, twistLimitDegrees=15}
 {bone1Name=dyn_loincloth_03_l, bone2Name=dyn_loincloth_02_l, swing1LimitDegrees=45, swing2LimitDegrees=15, twistLimitDegrees=15}
 {bone1Name=dyn_loincloth_02_l, bone2Name=dyn_loincloth_01_l, swing1LimitDegrees=45, swing2LimitDegrees=15, twistLimitDegrees=15}
 {bone1Name=dyn_loincloth_01_l, bone2Name=pelvis, swing1LimitDegrees=45, swing2LimitDegrees=15, twistLimitDegrees=15}
 {
  bone1Name=dyn_loincloth_02_l
  bone2Name=dyn_loincloth_02_r
  swing1LimitDegrees=45
  swing1Motion=Free
  swing2LimitDegrees=45
  swing2Motion=Free
  twistLimitDegrees=45
  twistMotion=Free
 }
 {bone1Name=dyn_bell_010, bone2Name=dyn_bell_09, swing1LimitDegrees=45, swing2LimitDegrees=45, twistLimitDegrees=45}
 {bone1Name=dyn_bell_01, bone2Name=dyn_bellBase, swing1LimitDegrees=45, swing2LimitDegrees=45, twistLimitDegrees=45}
 {bone1Name=dyn_bell_03, bone2Name=dyn_bell_01, swing1LimitDegrees=45, swing2LimitDegrees=45, twistLimitDegrees=45}
 {bone1Name=dyn_bell_05, bone2Name=dyn_bell_03, swing1LimitDegrees=45, swing2LimitDegrees=45, twistLimitDegrees=45}
 {bone1Name=dyn_bell_07, bone2Name=dyn_bell_05, swing1LimitDegrees=45, swing2LimitDegrees=45, twistLimitDegrees=45}
 {bone1Name=dyn_bell_09, bone2Name=dyn_bell_07, swing1LimitDegrees=45, swing2LimitDegrees=45, twistLimitDegrees=45}
 {
  bone1Name=dyn_bellBase
  bone2Name=pelvis
  swing1LimitDegrees=45
  swing1Motion=Free
  swing2LimitDegrees=45
  swing2Motion=Free
  twistLimitDegrees=45
  twistMotion=Free
 }
 {
  bone1Name=dyn_loincloth_05_r
  bone2Name=dyn_loincloth_05_l
  swing1LimitDegrees=45
  swing1Motion=Free
  swing2LimitDegrees=45
  swing2Motion=Free
  twistLimitDegrees=45
  twistMotion=Free
 }
]
