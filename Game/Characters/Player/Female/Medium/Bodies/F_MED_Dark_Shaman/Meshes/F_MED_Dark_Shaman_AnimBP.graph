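# Post-process style cosmetic AnimBP for the Dark Shaman female body, copying the parent mesh pose and adding secondary motion to loincloth, bell, bags and bones.
# Constraints and bone offsets, blended by crouch, arm-aim and thigh alphas, surround right-side RigidBody physics with skydiving force, then AnimDynamics chains.
# EventGraph computes those alphas and a delayed jump force for the bags.
# Left-side, loincloth and bell-chain RigidBodies, skirt-front constraints and computed skydivingForce never reach the output.
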AnimBlueprint /Game/Characters/Player/Female/Medium/Bodies/F_MED_Dark_Shaman/Meshes/F_MED_Dark_Shaman_AnimBP
target_skeleton=&/F_MED_Dark_Shaman_Skeleton
graphs [
 AnimGraph {nodes [
   n1 = "Constraint - dyn_skirtFront_01_r (RR)"!(ComponentPose, Alpha=1, ConstraintWeights_0=0, ConstraintWeights_1:float=1) -> (Pose) {
    alphaBoolBlend {blendOption=Linear}
    boneToModify {boneName=dyn_skirtFront_01_r}
    constraintSetup [{transformType=Rotation, targetBone {boneName=dyn_skirtFront_01_r}}, {transformType=Rotation, targetBone {boneName=thigh_twist_01_r}}]
    constraintWeights [1,1]
   }
   "Constraint - dyn_skirtFront_02_r (RR)"!(ComponentPose=@n1, Alpha=1, ConstraintWeights_0=0, ConstraintWeights_1:float=1) -> (Pose) {
    alphaBoolBlend {blendOption=Linear}
    boneToModify {boneName=dyn_skirtFront_02_r}
    constraintSetup [{transformType=Rotation, targetBone {boneName=dyn_skirtFront_02_r}}, {transformType=Rotation, targetBone {boneName=thigh_twist_01_r}}]
    constraintWeights [1,1]
   }
   n2 = Break Fort Anim Input Skydiving External Force!(FortAnimInput_SkydivingExternalForce=Get SkydivingExternalForceInput()) -> (
    SkydivingVectorForce=~
    HeadToPelvisDirection=~
    FloatingMultiplier=~
    FloatingAdditive=~
    DivingMultiplier=~
    DivingAdditive=~
    ParachutingMultiplier=~
    ParachutingAdditive=~
   )
   n3 = RigidBody(ExternalForce=@n2) {
    bClampLinearTranslationLimitToRefPose=true
    bOverrideWorldGravity=true
    overridePhysicsAsset=&/F_MED_DarkShaman_LSidePhysics
    simulationSpace=BaseBoneSpace
    alphaBoolBlend {blendOption=Linear}
    baseBoneRef {boneName=pelvis}
    componentLinearAccScale=<0.5 0.5 0.5>
    componentLinearVelScale=<0.2 0.2 0.2>
    overrideWorldGravity {z=-980}
   }
   RigidBody(ComponentPose=@n3, ExternalForce=@n2) {
    bClampLinearTranslationLimitToRefPose=true
    bOverrideWorldGravity=true
    overridePhysicsAsset=&/F_MED_DarkShaman_Loincloth_Physics
    simulationSpace=BaseBoneSpace
    alphaBoolBlend {blendOption=Linear}
    baseBoneRef {boneName=pelvis}
    componentAppliedLinearAccClamp=<1000 1000 1000>
    componentLinearAccScale=<0.5 0.5 0.5>
    componentLinearVelScale=<0.2 0.2 0.2>
    overrideWorldGravity {z=-980}
   }
   RigidBody() {
    bClampLinearTranslationLimitToRefPose=true
    bOverrideWorldGravity=true
    overridePhysicsAsset=&/F_MED_DarkShaman_BellChain_Physics
    simulationSpace=BaseBoneSpace
    alphaBoolBlend {blendOption=Linear}
    baseBoneRef {boneName=dyn_thigh_half_r}
    componentAppliedLinearAccClamp=<1000 1000 1000>
    componentLinearAccScale=<0.3 0.3 0.3>
    componentLinearVelScale=<0.1 0.1 0.1>
    overrideWorldGravity {z=-980}
   }
   n4 = "Constraint - dyn_loincloth_01_r (RRR)"!(
    ComponentPose=Local To Component(LocalPose=Copy Pose From Mesh() {bUseAttachedParent=true})
    Alpha=1
    ConstraintWeights_0=0.6
    ConstraintWeights_1:float=0.1
    ConstraintWeights_2:float=0.3
   ) -> (Pose) {
    alphaBoolBlend {blendOption=Linear}
    boneToModify {boneName=dyn_loincloth_01_r}
    constraintSetup [
     {transformType=Rotation, targetBone {boneName=thigh_twist_01_r}}
     {transformType=Rotation, targetBone {boneName=thigh_twist_01_l}}
     {transformType=Rotation, targetBone {boneName=pelvis}}
    ]
    constraintWeights [1,1,1]
   }
   n5 = "Constraint - dyn_loincloth_01_l (RRR)"!(ComponentPose=@n4, Alpha=1, ConstraintWeights_0=0.6, ConstraintWeights_1:float=0.1, ConstraintWeights_2:float=0.3) -> (Pose) {
    alphaBoolBlend {blendOption=Linear}
    boneToModify {boneName=dyn_loincloth_01_l}
    constraintSetup [
     {transformType=Rotation, targetBone {boneName=thigh_twist_01_l}}
     {transformType=Rotation, targetBone {boneName=thigh_twist_01_r}}
     {transformType=Rotation, targetBone {boneName=pelvis}}
    ]
    constraintWeights [1,1,1]
   }
   n6 = "Transform (Modify) Bone - Bone: dyn_loincloth_01_l"(ComponentPose=@n5, Alpha=Get crouchAlpha() -> (crouchAlpha:double), Translation="0,0,2", Rotation="0,0,-10", Scale="1,1,1") {
    rotationMode=BMM_Additive
    rotationSpace=BCS_ParentBoneSpace
    translationMode=BMM_Additive
    translationSpace=BCS_ParentBoneSpace
    alphaBoolBlend {blendOption=Linear}
    boneToModify {boneName=dyn_loincloth_01_l}
    rotation {roll=-10}
    translation {z=2}
   }
   n7 = "Constraint - dyn_thigh_half_r (P)"(ComponentPose=@n6, Alpha=1, ConstraintWeights_0=1) {
    alphaBoolBlend {blendOption=Linear}
    boneToModify {boneName=dyn_thigh_half_r}
    constraintSetup [{targetBone {boneName=thigh_r}}]
    constraintWeights [1]
   }
   n8 = "Constraint - dyn_bellBase (PP)"!(ComponentPose=@n7, Alpha=Get antiCrouchAlpha() -> (antiCrouchAlpha:double), ConstraintWeights_0=0.3, ConstraintWeights_1:float=0.7) -> (Pose) {
    alphaBoolBlend {blendOption=Linear}
    boneToModify {boneName=dyn_bellBase}
    constraintSetup [{targetBone {boneName=thigh_twist_01_r}}, {targetBone {boneName=pelvis}}]
    constraintWeights [1,1]
   }
   n9 = "Transform (Modify) Bone - Bone: dyn_bellBase"(ComponentPose=@n8, Alpha=Get crouchAlpha(), Translation="-3.623,-3.706,-4.130", Rotation="-17.146,6.164,39.136", Scale="1,1,1") {
    rotationMode=BMM_Replace
    rotationSpace=BCS_BoneSpace
    translationMode=BMM_Replace
    translationSpace=BCS_BoneSpace
    alphaBoolBlend {blendOption=Linear}
    boneToModify {boneName=dyn_bellBase}
    rotation=<rot -17.146 6.164 39.1359>
    translation=<-3.623 -3.706 -4.13>
   }
   n10 = "Transform (Modify) Bone - Bone: upperarm_shift_r"(ComponentPose=@n9, Alpha=Get armAim() -> (armAim:double), Translation="-5,-4,6", Rotation="0,0,0", Scale="1,1,1") {
    translationMode=BMM_Replace
    translationSpace=BCS_ParentBoneSpace
    alphaBoolBlend {blendOption=Linear}
    boneToModify {boneName=upperarm_shift_r}
    translation=<-5 -4 6>
   }
   n11 = RigidBody(ComponentPose=@n10, ExternalForce=@n2) {
    bClampLinearTranslationLimitToRefPose=true
    bOverrideWorldGravity=true
    overridePhysicsAsset=&/F_MED_DarkShaman_RSidePhysics
    simulationSpace=BaseBoneSpace
    alphaBoolBlend {blendOption=Linear}
    baseBoneRef {boneName=pelvis}
    componentAppliedLinearAccClamp=<1000 1000 1000>
    componentLinearAccScale=<0.3 0.3 0.35>
    componentLinearVelScale=<0.35 0.35 0.4>
    overrideWorldGravity {z=-980}
   }
   n12 = "Constraint - dyn_bonesBase (RR)"!(ComponentPose=@n11, Alpha=1, ConstraintWeights_0=0.5, ConstraintWeights_1:float=0.5) -> (Pose) {
    alphaBoolBlend {blendOption=Linear}
    boneToModify {boneName=dyn_bonesBase}
    constraintSetup [{transformType=Rotation, targetBone {boneName=dyn_bonesBase}}, {transformType=Rotation, targetBone {boneName=dyn_skirt_01_l}}]
    constraintWeights [1,1]
   }
   n13 = "Transform (Modify) Bone - Bone: dyn_bonesBase"(ComponentPose=@n12, Alpha=Get crouchAlpha(), Translation="5.567,-2.011,-12.779", Rotation="-145,0,-90", Scale="1,1,1") {
    rotationMode=BMM_Replace
    rotationSpace=BCS_ParentBoneSpace
    translationMode=BMM_Replace
    translationSpace=BCS_ParentBoneSpace
    alphaBoolBlend {blendOption=Linear}
    alphaScaleBiasClamp {bInterpResult=true}
    boneToModify {boneName=dyn_bonesBase}
    rotation {pitch=-145, roll=-90}
    translation=<5.567 -2.011 -12.779>
   }
   n14 = "dyn_bonesBase_02.rotateZ = thigh_l.rotateZ - Bone Driven Controller"!(ComponentPose:Component Space Pose Link Structure=@n13, Alpha:float=Get antiCrouchAlpha()) -> (Pose:Component Space Pose Link Structure) {
    is=title AnimGraphNode_BoneDrivenController
    bAffectTargetRotationZ=true
    sourceComponent=RotationZ
    alphaBoolBlend {blendOption=Linear}
    sourceBone {boneName=thigh_l}
    targetBone {boneName=dyn_bonesBase_02}
   }
   n15 = Get jumping_ext_force() -> (jumping_ext_force:Vector)
   n16 = Anim Dynamics - Bone: dyn_bag_01(ComponentPose=@n14, Alpha=1, ExternalForce=@n15) {
    angularDampingOverride=1
    angularSpringConstant=250
    bAngularSpring=true
    bGravityOverrideInSimSpace=true
    bOverrideAngularDamping=true
    simulationSpace=RootRelative
    alphaBoolBlend {blendOption=Linear}
    boundBone {boneName=dyn_bag_01}
    componentAppliedLinearAccClamp=<1000 1000 1000>
    componentLinearAccScale=<0.6 0.6 0.6>
    componentLinearVelScale=<0.4 0.4 0.4>
    physicsBodyDefinitions [{
      sphereCollisionRadius=0
      boundBone {boneName=dyn_bag_01}
      boxExtents=<8 8 16>
      constraintSetup {angularLimitsMax=<10 10 2>, angularLimitsMin=<-10 -10 -45>, angularTarget=<0 0 0>}
      localJointOffset=<6 0 0>
     }]
   }
   n17 = Anim Dynamics - Bone: dyn_bag_02(ComponentPose=@n16, Alpha=1, ExternalForce=@n15) {
    angularDampingOverride=1
    angularSpringConstant=250
    bAngularSpring=true
    bGravityOverrideInSimSpace=true
    bOverrideAngularDamping=true
    simulationSpace=RootRelative
    alphaBoolBlend {blendOption=Linear}
    boundBone {boneName=dyn_bag_02}
    componentAppliedLinearAccClamp=<1000 1000 1000>
    componentLinearAccScale=<0.6 0.6 0.6>
    componentLinearVelScale=<0.4 0.4 0.4>
    physicsBodyDefinitions [{
      sphereCollisionRadius=0
      boundBone {boneName=dyn_bag_02}
      boxExtents=<8 8 16>
      constraintSetup {angularLimitsMax=<10 10 2>, angularLimitsMin=<-10 -10 -45>, angularTarget=<0 0 0>}
      localJointOffset=<6 0 0>
     }]
    relativeSpaceBone {boneName=dyn_bag_01}
   }
   n18 = Anim Dynamics - Bone: dyn_bones_02(ComponentPose=@n17, Alpha=1, ExternalForce="0,0,0") {
    angularDampingOverride=1
    angularSpringConstant=250
    bAngularSpring=true
    bGravityOverrideInSimSpace=true
    bOverrideAngularDamping=true
    simulationSpace=RootRelative
    alphaBoolBlend {blendOption=Linear}
    boundBone {boneName=dyn_bones_02}
    componentAppliedLinearAccClamp=<1000 1000 1000>
    componentLinearAccScale=<0.6 0.6 0.6>
    componentLinearVelScale=<0.4 0.4 0.4>
    physicsBodyDefinitions [{
      sphereCollisionRadius=0
      boundBone {boneName=dyn_bones_02}
      boxExtents=<10 10 20>
      constraintSetup {angularLimitsMax=<1 8 1>, angularLimitsMin=<-1 -8 -15>, angularTarget=<0 0 0>}
      localJointOffset=<8 5 0>
     }]
   }
   n19 = Anim Dynamics - Bone: dyn_bones_03(ComponentPose=@n18, Alpha=0.5, ExternalForce="0,0,0") {
    angularDampingOverride=1
    angularSpringConstant=250
    bAngularSpring=true
    bGravityOverrideInSimSpace=true
    bOverrideAngularDamping=true
    simulationSpace=RootRelative
    alphaBoolBlend {blendOption=Linear}
    boundBone {boneName=dyn_bones_03}
    componentAppliedLinearAccClamp=<1000 1000 1000>
    componentLinearAccScale=<0.6 0.6 0.6>
    componentLinearVelScale=<0.4 0.4 0.4>
    physicsBodyDefinitions [{
      sphereCollisionRadius=0
      boundBone {boneName=dyn_bones_03}
      boxExtents=<10 10 20>
      constraintSetup {angularLimitsMax=<1 8 1>, angularLimitsMin=<-1 -8 -15>, angularTarget=<0 0 0>}
      localJointOffset=<8 5 0>
     }]
   }
   n20 = Anim Dynamics - Bone: dyn_bones_04(ComponentPose=@n19, Alpha=0.5, ExternalForce="0,0,0") {
    angularDampingOverride=1
    angularSpringConstant=250
    bAngularSpring=true
    bGravityOverrideInSimSpace=true
    bOverrideAngularDamping=true
    simulationSpace=RootRelative
    alphaBoolBlend {blendOption=Linear}
    boundBone {boneName=dyn_bones_04}
    componentAppliedLinearAccClamp=<1000 1000 1000>
    componentLinearAccScale=<0.6 0.6 0.6>
    componentLinearVelScale=<0.4 0.4 0.4>
    physicsBodyDefinitions [{
      sphereCollisionRadius=0
      boundBone {boneName=dyn_bones_04}
      boxExtents=<10 10 20>
      constraintSetup {angularLimitsMax=<1 8 1>, angularLimitsMin=<-1 -8 -15>, angularTarget=<0 0 0>}
      localJointOffset=<8 5 0>
     }]
   }
   n21 = "Transform (Modify) Bone - Bone: dyn_bonesBase"(ComponentPose=@n20, Alpha=Get thigh_shape_lt_l() -> (thigh_shape_lt_l:double), Translation="0,0,0", Rotation="0,60,0", Scale="1,1,1") {
    rotationMode=BMM_Additive
    rotationSpace=BCS_BoneSpace
    scaleMode=BMM_Additive
    scaleSpace=BCS_BoneSpace
    translationMode=BMM_Additive
    translationSpace=BCS_BoneSpace
    alphaBoolBlend {blendOption=Linear}
    boneToModify {boneName=dyn_bonesBase}
    rotation {yaw=60}
   }
   Output Pose(Result=Component To Local(ComponentPose=@n21)) {layerGroup=None}
   Get skydivingForce() -> (skydivingForce:Vector)
  ]}
 EventGraph {nodes [
   Event BlueprintUpdateAnimation()
   > n1 = Sequence!(execute) -> (then_0, then_1, then_2:Exec, then_3:Exec, then_4:Exec)
   > n2 = Branch(Condition=Get bIsJumping())
   > Delay(Duration=0.25)
   > Set jumping_force_vector(jumping_force_vector:Vector="0.25,2.5,0.1") -> (Output_Get:Vector)
   > Set jumping_ext_force(jumping_ext_force:Vector="0,0,2000") -> (Output_Get:Vector)
   > n3 = Delay(Duration=0.875)
   Do Once(execute=@n2.else, Reset=@n3.then)
   > Set jumping_ext_force(jumping_ext_force="0, 0, 0")
   Set thigh_shape_lt_l(
    execute=@n1.then_4
    thigh_shape_lt_l:double=GetAlphaFromReferenceTransformDelta(
     SkeletalMeshComponent=Get Skel_Mesh()
     BoneName=thigh_l
     BaseName=pelvis
     DeltaComponentType=EulerY
     InputRange=(X=140,Y=15)
     OutputRange=(X=1,Y=0)
     TwistAxis="0,0,1"
    )
   ) -> (Output_Get:double)
   n4 = GetTwistAndSwingAngleOfDeltaRotationFromRefPose(self=GetOwnerSkeletalMeshComponent(), BoneName=upperarm_r)
   n5 = Set armTwist(execute=@n1.then_3, armTwist:double=MapRangeClamped(Value=@n4, InRangeA=12, InRangeB=60)) -> (Output_Get:double)
   > n6 = Set armSwing(armSwing:double=MapRangeClamped(Value=@n4.OutSwingAngle, InRangeA=61, InRangeB=126)) -> (Output_Get:double)
   > Set armAim(armAim:double=float / float(A=float + float(A=@n5, B=@n6) {is=title K2Node_CommutativeAssociativeBinaryOperator}, B=2) {is=title K2Node_CallFunction}) -> (Output_Get:double)
   n7 = Get PawnVelocityZ()
   n8 = RandomFloatInRangeFromStream!(
    Stream:"Random Stream Structure (by ref)"=Get random_stream() -> (random_stream:Random Stream Structure)
    Min:float=float * float(A=@n7, B=2.5)
    Max:float=float * float(A=@n7, B=5.5)
   ) -> (ReturnValue:float)
   n9 = SelectFloat(
    A=Make Perlin Noise and Remap!(Value:float=@n8, RangeOutMin:float=-10000, RangeOutMax:float=15000) -> (ReturnValue:float)
    B=0
    bPickA=AND Boolean(A="float < float"!(A:double=@n8, B:double=0) -> (ReturnValue:bool) {is=title K2Node_CallFunction}, B=Get bIsSkyDiving())
   )
   n10 = float * float(A=@n9, B=0.1)
   n11 = Get bIsParachuteOpened()
   Set skydivingForce(execute=@n1.then_2, skydivingForce:Vector=MakeVector(Y=0, Z=SelectFloat(A=@n10, B=@n9, bPickA=@n11))) -> (Output_Get:Vector)
   n12 = Get bIsCrouching()
   Set crouchAlpha(execute=@n1.then_1, crouchAlpha:double="To Float (Boolean)"(InBool=@n12)) -> (Output_Get:double)
   > Set antiCrouchAlpha(antiCrouchAlpha:double="To Float (Boolean)"(InBool=NOT Boolean(A=@n12))) -> (Output_Get:double)
   Event BlueprintInitializeAnimation()
   > Set Skel_Mesh(Skel_Mesh=GetOwnerSkeletalMeshComponent())
   Get PawnVelocityZ()
   Set random_stream(random_stream:Random Stream Structure=MakeRandomStream!(InitialSeed:int=0) -> (ReturnValue:Random Stream Structure)) -> (Output_Get:Random Stream Structure)
  ]}
]
variables [
 antiCrouchAlpha {type=number}
 armAim {type=number}
 armSwing {type=number}
 armTwist {type=number}
 crouchAlpha {type=number}
 jumping_ext_force {type=Vector}
 jumping_force_vector {type=Vector}
 random_stream {type=RandomStream, default {initialSeed=0, seed=0}}
 Skel_Mesh {type=SkeletalMeshComponent}
 skydivingForce {type=Vector}
 thigh_shape_lt_l {type=number}
]
